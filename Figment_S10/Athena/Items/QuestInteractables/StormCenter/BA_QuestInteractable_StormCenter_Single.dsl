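; Quest interactable Blueprint for a storm center marker that players touch for objective progress, disabled in creative mode.
; The server credits each player once per circle via a quest stat event and resets on storm circle updates.
; Clients show it only for active or party-shared quests, spinning its crosshair nearby and playing success effects on quest updates.
; RepNotifies drive relocation and fading, though the fade-out branch only disables collision.
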(event EventBeginPlay
  (bind _debugon (Variables|Default|GetDebugOn))
  (bind _debugon_2 (Variables|Default|GetDebugOn))
  ;[1 0.1332 0] TEMP - need something better from code here in future
  (Utilities|FlowControl|Delay 5.0)
  ;[1 1 1] CLIENT AND SERVER - Check if In Creative Mode or not
  (bind _asfort_game_state_athena (Utilities|Casting|CastToFortGameStateAthena (Game|GetGameState)))
  (if (Variables|/Script/FortniteGame.FortGameStateAthena|GetbPlaylistDataIsLoaded _asfort_game_state_athena)
    (bind _cancompletequest (Class|AthenaQuestFunctionLibrary|AthenaCheckCreativeMode _asfort_game_state_athena))
    (if _cancompletequest
      (Variables|Default|SetCreativeMode false)
      (Development|PrintString "Creative Playlist Loaded" _debugon_2 _debugon_2)
      (else
        (Variables|Default|SetCreativeMode true)
        (Development|PrintString "Creative Playlist NOT Loaded" _debugon_2 _debugon_2)))
    (Networking|SwitchHasAuthority
      (:Authority
        ;[1 1 1] SERVER ONLY - Bind Event for Volume Touch
        (Collision|BindEventtoOnComponentBeginOverlap (Variables|Default|GetTouchCapsule) (EventDispatchers|CreateEvent :!event Custom|PawnTouchedCapsule))
        ;[1 1 1] SERVER ONLY - Bind Events for Updating StormCircle Position
        (Utilities|Casting|CastToFortGameModeAthena (Game|GetGameMode)))
      (:Remote
        ;[1 1 1] Is someone sharing this quest
        (bind (_bisassistingobjective _bisvalidplayercontroller) (Class|AthenaQuestFunctionLibrary|AthenaPartyAssistCheck (Fortnite|GetFirstLocalFortPlayerController true) (Variables|QuestStuff|GetBackendName)))
        (Variables|Default|SetIsQuestShared (and _bisassistingobjective _bisvalidplayercontroller))
        ;[1 1 1] CLIENT ONLY - Check if local client has quest or not
        (bind _questvalid (Quest|DoesPlayerHaveQuest (Utilities|Array|Get(acopy) (Fortnite|GetLocalFortPlayerControllers true)) (Variables|QuestStuff|GetQuestItem) (Variables|QuestStuff|GetBackendName)))
        (if (and _questvalid (not (Variables|Default|GetCreativeMode)))
          _
          (elif (Variables|Default|GetIsQuestShared)
            _
            (else
              (Variables|QuestStuff|SetQuestLocallyActive false)
              (CallFunction|HideQuestActor)
              (Development|PrintString "HIDE STORM CIRCLE ACTOR" _debugon _debugon "(R=1.000000,G=0.150610,B=0.950520,A=1.000000)" 1e+01)
              ;[1 1 1] CLIENT ONLY - Clients do not need this collision running as it is for quest completion logic and the only thing that matters is what the server says.
              (Collision|SetCollisionEnabled (Variables|Default|GetTouchCapsule) "NoCollision"))))
        (Variables|QuestStuff|SetQuestLocallyActive true)
        (CallFunction|InitializeStormCenterActor)
        (Development|PrintString "INITIALIZE" _debugon _debugon "(R=1.000000,G=0.378307,B=0.986602,A=1.000000)" 1e+01)))
    (else
      (Playlist|BindEventtoOnCurrentPlaylistLoaded _asfort_game_state_athena (delegate Custom|PlaylistLoaded)))))

(event Custom|PlaylistLoaded
  (params (PlaylistName :type Name) (PlaylistContextTags :type Gameplay%20Tag%20Container%20Structure%20%28by%20ref%29))
  (bind _debugon (Variables|Default|GetDebugOn))
  (bind _debugon_2 (Variables|Default|GetDebugOn))
  ;[1 1 1] CLIENT AND SERVER - Check if In Creative Mode or not
  (bind _cancompletequest (Class|AthenaQuestFunctionLibrary|AthenaCheckCreativeMode (Game|GetGameState)))
  (if _cancompletequest
    (Variables|Default|SetCreativeMode false)
    (Development|PrintString "Creative Playlist Loaded" _debugon_2 _debugon_2)
    (else
      (Variables|Default|SetCreativeMode true)
      (Development|PrintString "Creative Playlist NOT Loaded" _debugon_2 _debugon_2)))
  (Networking|SwitchHasAuthority
    (:Authority
      ;[1 1 1] SERVER ONLY - Bind Event for Volume Touch
      (Collision|BindEventtoOnComponentBeginOverlap (Variables|Default|GetTouchCapsule) (EventDispatchers|CreateEvent :!event Custom|PawnTouchedCapsule))
      ;[1 1 1] SERVER ONLY - Bind Events for Updating StormCircle Position
      (Utilities|Casting|CastToFortGameModeAthena (Game|GetGameMode)))
    (:Remote
      ;[1 1 1] Is someone sharing this quest
      (bind (_bisassistingobjective _bisvalidplayercontroller) (Class|AthenaQuestFunctionLibrary|AthenaPartyAssistCheck (Fortnite|GetFirstLocalFortPlayerController true) (Variables|QuestStuff|GetBackendName)))
      (Variables|Default|SetIsQuestShared (and _bisassistingobjective _bisvalidplayercontroller))
      ;[1 1 1] CLIENT ONLY - Check if local client has quest or not
      (bind _questvalid (Quest|DoesPlayerHaveQuest (Utilities|Array|Get(acopy) (Fortnite|GetLocalFortPlayerControllers true)) (Variables|QuestStuff|GetQuestItem) (Variables|QuestStuff|GetBackendName)))
      (if (and _questvalid (not (Variables|Default|GetCreativeMode)))
        _
        (elif (Variables|Default|GetIsQuestShared)
          _
          (else
            (Variables|QuestStuff|SetQuestLocallyActive false)
            (CallFunction|HideQuestActor)
            (Development|PrintString "HIDE STORM CIRCLE ACTOR" _debugon _debugon "(R=1.000000,G=0.150610,B=0.950520,A=1.000000)" 1e+01)
            ;[1 1 1] CLIENT ONLY - Clients do not need this collision running as it is for quest completion logic and the only thing that matters is what the server says.
            (Collision|SetCollisionEnabled (Variables|Default|GetTouchCapsule) "NoCollision"))))
      (Variables|QuestStuff|SetQuestLocallyActive true)
      (CallFunction|InitializeStormCenterActor)
      (Development|PrintString "INITIALIZE" _debugon _debugon "(R=1.000000,G=0.378307,B=0.986602,A=1.000000)" 1e+01))))

(event Custom|InitializeStormCenterActor
  (bind _lightningboltmesh (Variables|Default|GetLightningboltMesh))
  ;[1 1 1] CLIENT ONLY - Initialize Event to List For Quest Update
  (bind _asfort_player_controller_athena (Utilities|Casting|CastToFortPlayerControllerAthena (Utilities|Array|Get(acopy) (Fortnite|GetLocalFortPlayerControllers true))))
  (Quests|BindEventtoOnQuestsUpdated (Quests|GetQuestManager _asfort_player_controller_athena "Invalid") (EventDispatchers|CreateEvent :!event Custom|QuestUpdatedAfterTouch))
  ;[1 1 1] CLIENT ONLY- Initialize Event for Spinning Crosshairs when close
  (Collision|BindEventtoOnComponentBeginOverlap (Variables|Default|GetStartSpinSphere) (EventDispatchers|CreateEvent :!event Custom|BeginCrosshairSpinning))
  (Collision|BindEventtoOnComponentEndOverlap (Variables|Default|GetEndSpinSphere) (EventDispatchers|CreateEvent :!event Custom|EndCrosshairSpinning))
  ;[1 1 1] CLIENT ONLY - Initialize Crosshair BP
  (Variables|RelocationHandling|SetInitialLocation (Variables|/Script/Engine.SceneComponent|GetRelativeLocation _lightningboltmesh))
  (Variables|SpinHandling|SetInitialRotation (Variables|/Script/Engine.SceneComponent|GetRelativeRotation _lightningboltmesh))
  (bind _returnvalue (Rendering|Material|CreateDynamicMaterialInstance (Variables|Default|GetLightningboltMesh) 0 "/Figment_S10/Athena/Items/QuestInteractables/StormCenter/FX/M_StormCenterMat.M_StormCenterMat" "None" :!class /Script/Engine.PrimitiveComponent))
  (Variables|FXandSound|SetLightningBoltMat _returnvalue)
  (bind _returnvalue_2 (Rendering|Material|CreateDynamicMaterialInstance (Variables|Default|GetCrosshairMesh) 0 "/Figment_S10/Athena/Items/QuestInteractables/StormCenter/FX/M_StormCenterMat.M_StormCenterMat" "None" :!class /Script/Engine.PrimitiveComponent))
  (Variables|FXandSound|SetCrosshairMat _returnvalue_2))

(event Custom|PawnTouchedCapsule
  (params (OverlappedComponent :type Primitive%20Component%20Object%20Reference) (OtherActor :type Actor%20Object%20Reference) (OtherComp :type Primitive%20Component%20Object%20Reference) (OtherBodyIndex :type Integer) (bFromSweep :type Boolean) (SweepResult :type Hit%20Result%20Structure%20%28by%20ref%29))
  ;[1 1 1] SERVER ONLY - Trigger StormCircle Marker
  (Utilities|IsValid OtherActor
    (:Is%20Valid
      (bind _asfort_player_controller (Utilities|Casting|CastToFortPlayerController (Pawn|GetController (Variables|/Script/Engine.Actor|GetInstigator OtherActor)))
        (:then
          (if (not (Utilities|Array|ContainsItem (Variables|QuestStuff|GetPlayersWhoHaveTouchedThisCenter) (Variables|/Script/Engine.Controller|GetPlayerState _asfort_player_controller)))
            (Utilities|Array|AddUnique (Variables|QuestStuff|GetPlayersWhoHaveTouchedThisCenter) (Variables|/Script/Engine.Controller|GetPlayerState _asfort_player_controller))
            (Quest|SendCustomStatEvent (Quests|GetQuestManager _asfort_player_controller "Invalid") (Variables|QuestStuff|GetObjectiveStat) 1)))
        (:CastFailed
          (Development|PrintString "Cast Invalid" true true "(R=1.000000,G=0.000000,B=0.044729,A=1.000000)"))))
    (:Is%20Not%20Valid
      (Development|PrintString "ActorInvalid" true true "(R=1.000000,G=0.000000,B=0.016047,A=1.000000)"))))

(event Custom|NewStormCircleEvent
  ;[1 1 1] SERVER - On StormCircle Update, reset ObjName and Stat
  (if (not (Variables|Default|GetCreativeMode))
    (CallFunction|StopSpinning)
    (Math|Integer|IncrementInt (Variables|QuestStuff|GetCurrentStormCircleNum)
      (:%20%20
        (Utilities|Array|Clear (Variables|QuestStuff|GetPlayersWhoHaveTouchedThisCenter))
        (Collision|SetCollisionEnabled (Variables|Default|GetTouchCapsule) "NoCollision")
        (Variables|RelocationHandling|SetFadeInOrOut true)
        (Development|PrintString "FADE OUT- CALL" (Variables|Default|GetDebugOn) (Variables|Default|GetDebugOn) "(R=0.000000,G=0.660000,B=1.000000,A=1.000000)" 1e+01)
        (Utilities|FlowControl|Delay (Variables|RelocationHandling|GetStormCircleFadeTime))
        ;[1 0.4933 0.0588] TEST BOOL
        (if (Variables|Default|GetTestingEnabled)
          _)))))

(event Custom|TestAdvanceStormCircle
  ;[1 1 1] SERVER - On StormCircle Update, reset ObjName and Stat
  ;[1 0.4933 0.0588] TEST BOOL
  (if (Variables|Default|GetTestingEnabled)
    (if (not (Variables|Default|GetCreativeMode))
      (CallFunction|StopSpinning)
      (Math|Integer|IncrementInt (Variables|QuestStuff|GetCurrentStormCircleNum)
        (:%20%20
          (Utilities|Array|Clear (Variables|QuestStuff|GetPlayersWhoHaveTouchedThisCenter))
          (Collision|SetCollisionEnabled (Variables|Default|GetTouchCapsule) "NoCollision")
          (Variables|RelocationHandling|SetFadeInOrOut true)
          (Development|PrintString "FADE OUT- CALL" (Variables|Default|GetDebugOn) (Variables|Default|GetDebugOn) "(R=0.000000,G=0.660000,B=1.000000,A=1.000000)" 1e+01)
          (Utilities|FlowControl|Delay (Variables|RelocationHandling|GetStormCircleFadeTime))
          ;[1 0.4933 0.0588] TEST BOOL
          (if (Variables|Default|GetTestingEnabled)
            _))))))

(event Custom|QuestUpdatedAfterTouch
  (bind _returnvalue (Transformation|GetActorLocation))
  ;[1 1 1] LOCAL CLIENT ONLY - Visuals for Actor Disappearing after being touched -
  (Development|PrintString "Quest Updated" (Variables|Default|GetDebugOn) (Variables|Default|GetDebugOn) "(R=1.000000,G=0.000000,B=0.016047,A=1.000000)")
  (if (< (Variables|QuestStuff|GetLocalObjectiveCount) 0)
    (Variables|QuestStuff|SetLocalObjectiveCount 0)
    (CallFunction|StopSpinning)
    (Effects|SpawnEmitteratLocation (Variables|FXandSound|GetPartialSuccessEmitter) _returnvalue)
    (Audio|PlaySoundatLocation (Variables|FXandSound|GetPartialSuccessSound) _returnvalue)
    (if (Quest|HasCompletedObjective (Quests|GetQuestManager (Utilities|Array|Get(acopy) (Fortnite|GetLocalFortPlayerControllers true)) "Invalid") (Variables|QuestStuff|GetQuestItem) (Variables|QuestStuff|GetObjectiveStat) :!class /Script/FortniteGame.FortQuestManager)
      (Variables|QuestStuff|SetQuestLocallyActive false)
      ;[1 1 1] FX for Objective Completed
      (bind (_direction _lightningscale _lightningfadeouttime _crosshairscale _crosshairfadetime) (Timeline|FinishQuestTimeline PlayFromStart)
        (:Update
          (Transformation|SetWorldScale3D (Variables|Default|GetLightningboltMesh) (Math|Vector|Lerp(Vector) "1.300000,1.300000,1.300000" "0.100000,0.100000,0.100000" _lightningscale))
          (Transformation|SetWorldScale3D (Variables|Default|GetCrosshairMesh) (Math|Vector|Lerp(Vector) "1.250000,1.250000,1.250000" "0.010000,0.010000,0.010000" _crosshairscale))
          (Rendering|Material|SetScalarParameterValue (Variables|FXandSound|GetLightningBoltMat) "Opacity" (Math|Float|Lerp 1.0 0.0 _lightningfadeouttime) :!class /Script/Engine.MaterialInstanceDynamic)
          (Rendering|Material|SetScalarParameterValue (Variables|FXandSound|GetCrosshairMat) "Opacity" (Math|Float|Lerp 1.0 0.0 _crosshairfadetime) :!class /Script/Engine.MaterialInstanceDynamic))
        (:Finished
          (Development|PrintString "FINISHED!" (Variables|Default|GetDebugOn) (Variables|Default|GetDebugOn) "(R=1.000000,G=0.768907,B=0.000000,A=1.000000)" 1e+01)
          (Effects|SpawnEmitterAttached :EmitterTemplate (Variables|FXandSound|GetCompleteSuccessEmitter) :AttachToComponent (Variables|/Script/Engine.Character|GetMesh (Game|Player|GetPlayerPawn (Utilities|Array|Get(acopy) (Fortnite|GetLocalFortPlayerControllers true)))) :AttachPointName "None" :Scale "1.000000,1.000000,1.000000" :LocationType "SnapToTarget" :bAutoDestroy true :PoolingMethod "None" :bAutoActivate true)
          (Audio|PlaySoundatLocation (Variables|FXandSound|GetCompleteSuccessSound) (Transformation|GetActorLocation))))
      (else
        ;[1 1 1] FX for Objective Updated
        (bind (_direction_2 _risetime _fadeouttime) (Timeline|DisappearTimeline PlayFromStart)
          (:Update
            (Transformation|SetRelativeLocation (Variables|Default|GetLightningboltMesh) (Math|Vector|Lerp(Vector) (Variables|RelocationHandling|GetInitialLocation) (+ (Variables|RelocationHandling|GetInitialLocation) "0, 0,1000.000000") _risetime) false false)
            (Rendering|Material|SetScalarParameterValue (Variables|FXandSound|GetLightningBoltMat) "Opacity" (Math|Float|Lerp 1.0 0.0 _fadeouttime) :!class /Script/Engine.MaterialInstanceDynamic))
          (:Finished
            (Effects|SpawnEmitterAttached :EmitterTemplate (Variables|FXandSound|GetPartialSuccessEmitter) :AttachToComponent (Variables|/Script/Engine.Character|GetMesh (Game|Player|GetPlayerPawn (Utilities|Array|Get(acopy) (Fortnite|GetLocalFortPlayerControllers true)))) :AttachPointName "None" :Scale "1.000000,1.000000,1.000000" :LocationType "SnapToTarget" :bAutoDestroy true :PoolingMethod "None" :bAutoActivate true)))))
    ;[1 1 1] HIDE QUEST ACTOR
    (CallFunction|HideQuestActor)))

(event Custom|FadeOutStormCenter
  ;[1 1 1] CLIENT ONLY - Visual Stuff
  ;;[1 1 1] QuickFadeOut
  (if (Variables|QuestStuff|GetQuestLocallyActive)
    (Collision|SetActorEnableCollision false)))

(event Custom|FadeInStormCenter
  ;[1 1 1] CLIENT ONLY - Visual Stuff
  ;;[1 1 1] QuickFadeIn
  (if (Variables|QuestStuff|GetQuestLocallyActive)
    (CallFunction|RevealQuestActor)
    (Collision|SetActorEnableCollision true)
    (bind (_direction _fadetime) (Timeline|FadeIn PlayFromStart)
      (:Update
        (Rendering|Material|SetScalarParameterValue (Variables|FXandSound|GetLightningBoltMat) "Opacity" (Math|Float|Lerp 0.0 1.0 _fadetime) :!class /Script/Engine.MaterialInstanceDynamic)
        (Transformation|SetRelativeLocation (Variables|Default|GetLightningboltMesh) (Variables|RelocationHandling|GetInitialLocation) false false)))))

(event Custom|BeginCrosshairSpinning
  (params (OverlappedComponent :type Primitive%20Component%20Object%20Reference) (OtherActor :type Actor%20Object%20Reference) (OtherComp :type Primitive%20Component%20Object%20Reference) (OtherBodyIndex :type Integer) (bFromSweep :type Boolean) (SweepResult :type Hit%20Result%20Structure%20%28by%20ref%29))
  ;[1 1 1] CLIENT ONLY - Visual Stuff
  ;;[1 1 1] Enable spinning Crosshair
  (if (Variables|QuestStuff|GetQuestLocallyActive)
    (Development|PrintString "SPIIIIIIIIIIIIIIIIIIIIN!!!!" (Variables|Default|GetDebugOn) (Variables|Default|GetDebugOn) "(R=0.000000,G=0.660000,B=1.000000,A=1.000000)" 1e+01)
    (Components|Timeline|SetLooping (Variables|BA_QuestInteractable_StormCenter_Single|GetSpin) true)
    (bind (_direction _bob _spin) (Timeline|Spin PlayFromStart)
      (:Update
        (Transformation|SetWorldRotation (Variables|Default|GetCrosshairMesh) (Math|Rotator|Lerp(Rotator) "0, 0, 0" "0,720.000000, 0" _spin) false false)))))

(event Custom|EndCrosshairSpinning
  (params (OverlappedComponent :type Primitive%20Component%20Object%20Reference) (OtherActor :type Actor%20Object%20Reference) (OtherComp :type Primitive%20Component%20Object%20Reference) (OtherBodyIndex :type Integer))
  ;[1 1 1] CLIENT ONLY - Visual Stuff
  ;;[1 1 1] Disable spinning Crosshair
  (Utilities|FlowControl|Sequence
    (:then_0
      ;;[1 1 1] Enable spinning Crosshair
      (Timeline|Spin Stop))
    (:then_1
      (Variables|SpinHandling|SetCurrentRotation (Variables|/Script/Engine.SceneComponent|GetRelativeRotation (Variables|Default|GetCrosshairMesh)))
      (bind (_direction _droptime) (Timeline|Timeline_1 PlayFromStart)
        (:Update
          (Transformation|SetWorldRotation (Variables|Default|GetCrosshairMesh) (Math|Rotator|Lerp(Rotator) (Variables|SpinHandling|GetCurrentRotation) (Variables|SpinHandling|GetInitialRotation) _droptime) false false))))))

(event Custom|StopSpinning
  ;[1 1 1] CLIENT ONLY - Visual Stuff
  ;;[1 1 1] Disable spinning Crosshair
  (Timeline|Timeline_1 Stop))

(fn HideQuestActor
  (Networking|SwitchHasAuthority
    (:Remote
      (Collision|SetActorEnableCollision false)
      (Components|Activation|Deactivate (Variables|Default|GetLoopingAudio))
      (Rendering|SetActorHiddenInGame true)
      (Development|PrintString "STORM CIRCLE QUEST DEBUG: Quest Actor Hidden" false true "(R=0.331255,G=1.000000,B=0.000000,A=1.000000)" 1e+01))))

(fn OnRep_FadeOutBool
  (if (Variables|RelocationHandling|GetFadeInOrOut)
    (CallFunction|FadeOutStormCenter)
    (else
      (CallFunction|FadeinStormCenter))))

(fn OnRep_PlayersWhoHaveTouchedThisCenter)

(fn OnRep_ServerCurrentLocation
  (Transformation|SetActorLocation (Variables|RelocationHandling|GetServerCurrentLocation)))

(fn RevealQuestActor
  (Networking|SwitchHasAuthority
    (:Remote
      (Collision|SetActorEnableCollision true)
      (Rendering|SetActorHiddenInGame false)
      (Components|Activation|Activate (Variables|Default|GetLoopingAudio) false)
      (Development|SetHiddeninGame (Variables|Default|GetCrosshairMesh) false false)
      (Development|PrintString "STORM CIRCLE QUEST DEBUG: Fade IN Action" false true "(R=0.331255,G=1.000000,B=0.000000,A=1.000000)" 1e+01))))

(fn ConstructionScript)
